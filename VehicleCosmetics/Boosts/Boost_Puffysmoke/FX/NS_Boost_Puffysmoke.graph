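# Niagara system for the Puffysmoke vehicle boost cosmetic, a single sprite emitter that trails smoke puffs from the vehicle's boost ports.
# Particles are partitioned across filtered skeletal mesh sockets and shot along a fixed velocity, with spawn rate and lifetime varied by game-context masks.
# Puffs flash HDR-hot, then darken and fade while growing and animating SubUVs, under gravity, drag, collision and fading inherited movement.
# The painted-variant override module is disabled.
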NiagaraSystem /VehicleCosmetics/Boosts/Boost_Puffysmoke/FX/NS_Boost_Puffysmoke
dependencies {
 usedDataInterfaces [
  &NiagaraDataInterfaceCurve
  &NiagaraDataInterfaceColorCurve
  &NiagaraDataInterfaceSpriteRendererInfo
  &NiagaraDataInterfaceSkeletalMesh
 ]
 usedDynamicInputs []
 usedModules [
  &self:SystemSpawnScript.SystemScriptSource.SystemScriptGraph.NiagaraNodeAssignment_0.SetVariables_#1
  &self:SystemSpawnScript.SystemScriptSource.SystemScriptGraph.NiagaraNodeAssignment_1.SetVariables_#2
  &SystemState
 ]
}
emitters {NE_VC_Sprite_Parent_Imulsion1001 {
  emitterUpdateScript [
   &EmitterState(Life Cycle Mode=::NewEnumerator1, Loop Duration=1, UseLoopDelay=false)
   &VC_FilteredSocketCount()
   &SpawnRate(
    SpawnRate=&Multiply_Float(
     A=150
     B=/VehicleCosmetics/FX/Niagara/Modules/VC_Mask_Floats(Coro Value:NiagaraFloat=6, FrontEnd Value:NiagaraFloat=2, FN Value:NiagaraFloat=2)
    )
    Use Spawn Probability=false
   )
  ]
  particleSpawnScript [
   &VC_PaintedVariantsOverride() {enabled=false}
   &InitializeParticle(
    Lifetime=&Multiply_Float(
     A=&RandomRangeFloat(Minimum=0.35, Maximum=0.45)
     B=&VC_Mask_Floats(Coro Value:NiagaraFloat=0.85, FrontEnd Value:NiagaraFloat=1, FN Value:NiagaraFloat=0.5)
    )
    Sprite Rotation Angle=&RandomRangeFloat(Minimum=160, Maximum=220)
    Sprite Rotation Mode=::NewEnumerator0
    Sprite Size Mode=::NewEnumerator1
    Uniform Sprite Size Max=60
    Uniform Sprite Size Min=35
    UsePositionOffset=false
    Write Material Random=true
   )
   &PartitionParticles(Number Of Partitions=$Emitter.PortCount)
   &SkeletalMeshLocation(
    Bone / Socket Index=$Particles.PartitionParticles.Partition
    Bone / Socket Sampling Mode=::NewEnumerator5
    MaskSpawnGroup=false
    Mesh Sampling Type=::NewEnumerator6
    Skeletal Mesh=$System.SkeletalMesh
   )
   &AddVelocity(
    Velocity=&MakeCustomVectorFromBool(Bool=true, True Vector=<2000 25 -50>, False Vector=<2000 25 -50>)
    Velocity Speed Scale=-1
   )
   /VehicleCosmetics/FX/VC_FXParents/Modules/DLR_ProductPreviews(GameModeID:NiagaraInt32=$System.GameModeID, bPreviewVehicleSpeed:NiagaraBool=$System.bPreviewVehicleSpeed)
  ]
  particleUpdateScript [
   &ParticleState()
   &ScaleSpriteSize(Uniform Curve Sprite Scale=DataInterface() {propertyValues {Curve {keys [{tangentMode=RCTM_User, time=0, value=0.9}, {tangentMode=RCTM_User, time=1, value=1}]}}}) {module=ScaleSpriteSize001}
   &Collision(
    Age Colliding Particles=false
    KillOnCollision=false
    Module.Control Rest State Via Surface Normal=false
    Module.EnableMaxCollisionCount=false
    Module.Randomize Collision Normal Vector=false
    Particle Radius Scale=0.25
    Restitution=0
   )
   &ScaleColor(
    Scale Alpha=&FloatFromCurve(
     FloatCurve=DataInterface() {propertyValues {Curve {keys [
         <at 0 1 RCIM_Cubic>
         {
          arriveTangent=0.0182
          arriveTangentWeight=0.8096
          interpMode=RCIM_Cubic
          tangentMode=RCTM_Break
          tangentWeightMode=RCTWM_WeightedBoth
          time=1
          value=0.01
         }
        ]}}}
     Scale Curve=1.5
    )
    Scale RGB=&MakeVectorFromLinearColorRGB(Color=&ColorFromCurve(ColorCurve=DataInterface() {propertyValues {
        AlphaCurve {keys [<at 0 1 RCIM_Cubic>,<at 0.7168 0.1247>,<at 1 0 RCIM_Cubic>]}
        BlueCurve {keys [
          <at 0 1599.9999>
          <at 0.0846 0.8>
          {arriveTangent=2.7455, interpMode=RCIM_Cubic, leaveTangent=2.7455, time=0.1366, value=2}
          <at 0.1574 1>
          {arriveTangent=-3.3915, interpMode=RCIM_Cubic, leaveTangent=-3.3915, time=0.2407, value=0.64}
          <at 0.3343 0.4>
          <at 1 0.0783>
         ]}
        GreenCurve {keys [
          <at 0 2199.9998>
          <at 0.0846 1.1>
          {arriveTangent=0.9152, interpMode=RCIM_Cubic, leaveTangent=0.9152, time=0.1366, value=2.75}
          <at 0.1574 1.1667>
          {arriveTangent=-3.2031, interpMode=RCIM_Cubic, leaveTangent=-3.2031, time=0.2407, value=0.944}
          <at 0.3343 0.6>
          <at 1 0.1147>
         ]}
        RedCurve {keys [
          <at 0 4000>
          <at 0.0846 2>
          <at 0.1366 5 RCIM_Cubic>
          <at 0.1574 2>
          {arriveTangent=-5.6525, interpMode=RCIM_Cubic, leaveTangent=-5.6525, time=0.2407, value=1.6}
          <at 0.3343 1>
          <at 1 0.1875>
         ]}
       }}))
    ScaleA=true
    ScaleRGB=true
   )
   &ScaleSpriteSize(Uniform Curve Sprite Scale=DataInterface() {propertyValues {Curve {keys [<at 0 1 RCIM_Cubic>, {arriveTangent=1.3, interpMode=RCIM_Cubic, leaveTangent=1.3, time=0.02, value=1.5}, <at 1 2.3 RCIM_Cubic>]}}})
   &GravityForce(Gravity=<0 0 -250>)
   &SubUVAnimation(
    End Frame Range Override=40
    "Sprite Renderer "=DataInterface() {propertyValues {SpriteRenderer=&self:NE_VC_Sprite_Parent_Imulsion1001_0.NiagaraSpriteRendererProperties_1}}
    SubUV Animation Mode:ENiagara_SubUVLookupModeV2=::NewEnumerator2
    UseEndFrame=true
    UseStartFrame=true
   )
   /Niagara/Modules/Events/GenerateCollisionEvent(Module.EventDelay=true, Module.Use Velocity Threshold=true)
   &Drag(Use Linear Drag=true, Use Rotational Drag=true)
   &InheritSourceMovement(
    Applied Position Scale=&VectorFromFloat(Value=&FloatFromCurve(FloatCurve=DataInterface() {propertyValues {Curve {
         defaultValue=1
         keys [
          {arriveTangent=-1.7885, interpMode=RCIM_Cubic, leaveTangent=-1.7885, tangentMode=RCTM_User, time=0.05, value=1}
          {interpMode=RCIM_Cubic, tangentMode=RCTM_User, time=1, value=0}
         ]
        }}}))
    ApplyForce=false
    ApplyPosition=true
    ApplyPositionFalloff=false
   )
   &SolveForcesAndVelocity(Clamp Velocity=false, Limit Acceleration=false)
   &DynamicMaterialParameters(
    Index 0 Param 1=1
    Index 0 Param 2=5.5
    Index 0 Param 3=50
    Index 1 Param 1=1
    Index 1 Param 4=2
    Module.Index0Float0Write=true
    Module.Index0Float1Write=true
    Module.Index0Float2Write=true
    Module.Index0Float3Write=true
    Module.Index1Float0Write=true
    Module.Index1Float1Write=true
    Module.Index1Float2Write=true
    Module.Index1Float3Write=true
    Write Parameter Index 1=::NewEnumerator1
   )
  ]
  rendererClasses [&NiagaraSpriteRendererProperties]
  renderers [{rendererClass=&NiagaraSpriteRendererProperties}]
 }}
summary {
 systemName=NS_Boost_Puffysmoke
 emitters [{emitterName=NE_VC_Sprite_Parent_Imulsion1001, rendererClasses [&NiagaraSpriteRendererProperties]}]
 userVariables [
  User.bIsFrontendPreview {type=NiagaraBool}
  User.bIsPainted {type=NiagaraBool}
  User.blsIsCoroColor {type=NiagaraBool}
  User.bPreviewVehicleSpeed {type=NiagaraBool}
  User.ColorVariantID {type=NiagaraInt32}
  User.DTCarOffset {type=Vector3f, description="", defaultValue {struct=Vector3f, value=<0 0 0>}}
  User.GameModeID {type=NiagaraInt32, defaultValue=3}
  User.PaintedColor {type=LinearColor, description="", defaultValue {struct=LinearColor, value=<rgba 1 1 1 1>}}
  User.PaintedScalar {type=NiagaraFloat, defaultValue=1, description=""}
  User.PortCount {type=NiagaraInt32, defaultValue=2}
 ]
}
